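# Unlit, two-sided translucent Niagara material for a smoky swirl effect on the Linguini glider.
# Dynamic-parameter tiling and offsets drive a smoke texture, summed with a panning turbulence texture and inverted, time-evolving gradient noise.
# That mask, blended along V, indexes a blue glow curve atlas for emissive color.
# Opacity combines particle alpha, a V-gradient falloff and depth fade; several branches, including normals and vector noise, appear unused.
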Material /BRCosmetics/Gadgets/Assets/VinderTech_GliderChute/Glider_Linguini_Male/FX/M_SmokySwirl_Linear
graph {
 properties {
  bCastRayTracedShadows=false
  bUsedWithNiagaraMeshParticles=true
  bUsedWithNiagaraSprites=true
  blendMode=BLEND_Translucent
  pixelDepthOffsetMode=PDOM_Legacy
  shadingModel=MSM_Unlit
  translucencyLightingMode=TLM_VolumetricDirectional
  twoSided=true
  usedShadingModels=Unlit
 }
 nodes [
  n1 = TextureCoordinate()
  n2 = MaterialFunctionCall(Float2=@n1) {materialFunction=&BreakOutFloat2Components}
  n3 = SmoothStep(Value=Power(Base=@n2.G) {constExponent=0.5}) {constMin=0.6}
  n4 = DynamicParameter() {parameterIndex=3, defaultValue=<rgba 1 0 0.6 2>, paramNames [Depth Fade, Center V Coord, Center V Width, N/A]}
  DepthFade(Opacity=Multiply(A=ParticleColor().A, B=OneMinus(None=@n3)), FadeDistance=@n4.Depth Fade)
  n5 = DynamicParameter() {parameterIndex=2, defaultValue=<rgba 1.5 2 1 1>, paramNames [Base Sprite Tile U, Base Sprite Tile V, Radial Mask Tile U, Radial Mask Tile V]}
  n6 = DynamicParameter() {
   parameterIndex=1
   defaultValue=<rgba 1 1 1 2>
   paramNames [Base Sprite Offset U, Base Sprite Offset V, Radial Mask Offset U, Radial Mask Offset V]
  }
  n7 = AppendVector(A=ComponentMask(None=@n6.Base Sprite Offset U) {r=true}, B=ComponentMask(None=@n6.Base Sprite Offset V) {g=true})
  n8 = Constant2Vector() {g=1, r=1}
  n9 = DynamicParameter() {defaultValue=<rgba 1 1 1 2>, paramNames [Sprite Extrusion Speed, Sprite Spin Speed, Radial Mask Strength, Radial Mask Power]}
  n10 = Multiply(A=ComponentMask(None=@n8) {r=true}, B=@n9.Sprite Extrusion Speed)
  n11 = Multiply(A=ComponentMask(None=@n8) {g=true}, B=@n9.Sprite Spin Speed)
  n12 = Multiply(A=AppendVector(A=@n5.Base Sprite Tile U, B=@n5.Base Sprite Tile U), B=Add(A=AppendVector(A=@n10, B=@n11), B=Add(A=@n7, B=@n1)))
  n13 = TextureCoordinate()
  n14 = DDX(Value=@n13)
  n15 = DDY(Value=@n13)
  n16 = TextureSampleParameter2D!(UVs=@n12, "DDX(UVs)"=@n14, "DDY(UVs)"=@n15, Apply View MipBias) {
   constMipValue=0
   group=Base Sprite Texture
   mipValueMode=TMVM_Derivative
   parameterName=Base Sprite
   sortPriority=60
   texture=/Game/Effects/Fort_Effects/Textures/Tile/T_Inky_Smoke_Tile2
  }
  n17 = TextureSampleParameter2D(UVs=Panner(Coordinate=@n12) {speedY=0.6}, "DDX(UVs)"=@n14, "DDY(UVs)"=@n15) {
   constMipValue=0
   group=Base Sprite Texture
   mipValueMode=TMVM_Derivative
   parameterName=Secondary Sprite
   sortPriority=60
   texture=/Game/Global/Textures/Default/Caustics/T_Turbulence_Seamless
  }
  n18 = Add(A=@n17.R, B=Multiply(A=@n16.R, B=Constant() {r=1}))
  LinearInterpolate(A=@n16.R, B=@n18, Alpha=ScalarParameter() {defaultValue=1, parameterName=Add Secondary Sprite?})
  n19 = Time()
  n20 = Constant() {r=0.2}
  n21 = Noise!(World Position=AppendVector(A=@n12, B=Multiply(A=@n19, B=@n20)), FilterWidth) {bTiling=true, levelScale=4, levels=2, noiseFunction=NOISEFUNCTION_GradientALU, outputMin=0, repeatSize=10, scale=10}
  n22 = Power(Base=@n21) {constExponent=0.1}
  n23 = Add(A=Power(Base=@n18) {constExponent=0.1}, B=Multiply(A=OneMinus(None=@n22), B=Constant() {r=0.6}))
  n24 = Constant() {r=0.85}
  n25 = Add(A=@n4.Center V Coord, B=@n4.Center V Coord)
  n26 = CurveAtlasRowParameter(CurveTime=LinearInterpolate(A=Constant() {r=0.1}, B=Multiply(A=@n23, B=@n24), Alpha=SmoothStep(Min=@n4.Center V Coord, Max=@n25, Value=@n2.G) {constMax=0.1}) {constAlpha=0.6}) {
   atlas=&/CA_blueLight_glowy
   curve=&/C_blueLight_glowy
  }
  MaterialFunctionCall(Emissive=@n26) {materialFunction=&MF_FX_EmissiveAdjustments}
  n27 = WorldPosition()
  n28 = MaterialFunctionCall(X=Time()) {materialFunction=&MakeFloat3}
  Multiply(A=VectorNoise(World Position=Add(A=@n27.XYZ, B=@n28.Result)) {noiseFunction=VNF_VectorALU}, B=Constant() {r=33})
  Multiply()
  Power(Base=@n3, Exp=Constant() {r=2})
  TextureSampleParameter2D!(UVs=@n12, Apply View MipBias) {
   parameterName=normals
   samplerType=SAMPLERTYPE_Normal
   texture=/Game/Characters/Player/Male/Medium/Bodies/M_MED_Haunt/Textures/T_Thick_Smoke_05_N
  }
  Time()
 ]
}
