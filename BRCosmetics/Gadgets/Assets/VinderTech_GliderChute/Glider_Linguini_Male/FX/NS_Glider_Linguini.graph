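# Niagara burst effect for the Linguini glider cosmetic, built from five staggered burst emitters.
# A starburst and a radial sprite/mesh flash swell then collapse via size curves, with curved-noise smoke and velocity-aligned blue-to-green wisps around them.
# GPU sparks fly out in a cone, tinted cyan to yellow, enlarged by LOD distance and stretched by speed.
# Scratch modules seed smoke, wisps and sparks, taking camera position for the first two.
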NiagaraSystem /BRCosmetics/Gadgets/Assets/VinderTech_GliderChute/Glider_Linguini_Male/FX/NS_Glider_Linguini
dependencies {
 usedDataInterfaces [
  &NiagaraDataInterfaceColorCurve
  &NiagaraDataInterfaceCurve
  &NiagaraDataInterfaceMeshRendererInfo
  &NiagaraDataInterfaceCamera
 ]
 usedDynamicInputs []
 usedModules [&SystemState]
}
emitters {
 FN_Core_Smoke_Sprite {
  emitterUpdateScript [
   &EmitterState(Life Cycle Mode=::NewEnumerator1, Loop Behavior=::NewEnumerator1, Loop Delay=0.05, Loop Duration=0.1, UseLoopDelay=true)
   &SpawnBurst_Instantaneous(Spawn Count=50, Use Spawn Probability=false)
  ]
  particleSpawnScript [
   /SetVariables(Particles.InitSpriteRotRate=&RandomRangeFloat(Minimum=-2, Maximum=2)) {isSetParameters=true}
   &InitializeParticle(
    Color Maximum=<rgba 0.1255 0.1137 0.1059 1>
    Color Minimum=<rgba 0.2 0.1843 0.1686 1>
    Color Mode=::NewEnumerator2
    Lifetime Max=1.7
    Lifetime Min=1.2
    Lifetime Mode=::NewEnumerator1
    Sprite Rotation Angle=&RandomRangeFloat(Minimum=0, Maximum=360)
    Sprite Rotation Mode=::NewEnumerator0
    Sprite Size Mode=::NewEnumerator3
    Uniform Sprite Size=350
    UsePositionOffset=false
    Write Material Random=true
   )
   &self:ScratchModule(
    CameraPosition:Vector3f=&ReturnCameraVector(Camera Property=::NewEnumerator0, Coordinate Space=::Simulation)
    CenterScale=600
    velocity:Vector3f=<0 0 2500>
   ) {module=ScratchModule}
  ]
  particleUpdateScript [
   &ParticleState()
   &SpriteRotationRate(Rotation Rate=&Multiply_Float(
     A=$Particles.InitSpriteRotRate
     B=&FloatFromCurve(
      FloatCurve=DataInterface() {propertyValues {Curve {keys [
          {arriveTangent=-4.0107, interpMode=RCIM_Cubic, leaveTangent=-4.0107, tangentMode=RCTM_User, time=-0.0046, value=1.0113}
          {arriveTangent=-0.2496, interpMode=RCIM_Cubic, leaveTangent=-0.2496, tangentMode=RCTM_User, time=0.5, value=0.16}
         ]}}}
      CurveIndex=$Emitter.Age
      Scale Curve=150
     )
    ))
   &ScaleColor(
    Scale Alpha=&FloatFromCurve(
     FloatCurve=DataInterface() {propertyValues {Curve {keys [{arriveTangent=-2.749, interpMode=RCIM_Cubic, leaveTangent=-2.749, tangentMode=RCTM_User, time=0, value=1}, <at 1 0>]}}}
     Scale Curve=0.8
    )
    Scale RGB=&VectorFromFloat(Value=&FloatFromCurve(FloatCurve=DataInterface() {propertyValues {Curve {keys [{arriveTangent=1.8145, interpMode=RCIM_Cubic, leaveTangent=4.5795, tangentMode=RCTM_Break, time=0, value=0.4}, <at 0.3387 0.9565>, <at 1 1>]}}}))
    ScaleA=true
    ScaleRGB=true
   )
   &ScaleSpriteSize(Scale Factor=&Vector2DFromFloat(Value=&FloatFromCurve(
      FloatCurve=DataInterface() {propertyValues {Curve {keys [
          {arriveTangent=8.1876, interpMode=RCIM_Cubic, leaveTangent=5.3162, tangentMode=RCTM_User, time=0, value=0.15}
          {arriveTangent=1.6858, interpMode=RCIM_Cubic, leaveTangent=1.6858, tangentMode=RCTM_User, time=0.161, value=0.7059}
          {interpMode=RCIM_Cubic, tangentMode=RCTM_User, time=0.5, value=1}
          {arriveTangent=0.1638, interpMode=RCIM_Cubic, leaveTangent=0.1638, tangentMode=RCTM_User, time=1.1, value=0.4}
         ]}}}
      Scale Curve=&RandomRangeFloat(Minimum=1, Maximum=3)
     )))
   &DynamicMaterialParameters(
    Index 0 Param 1=&Add_Float(
     A=$Particles.RandomFloat
     B=&FloatFromCurve(FloatCurve=DataInterface() {propertyValues {Curve {keys [
          {arriveTangent=2.7533, interpMode=RCIM_Cubic, leaveTangent=2.847, tangentMode=RCTM_Break, time=0, value=0}
          {arriveTangent=1.3581, interpMode=RCIM_Cubic, leaveTangent=1.3581, tangentMode=RCTM_User, time=0.2558, value=0.5125}
          {arriveTangent=0.0529, interpMode=RCIM_Cubic, leaveTangent=0.2021, tangentMode=RCTM_User, time=1, value=1}
         ]}}})
    )
    Index 0 Param 3=&FloatFromCurve(FloatCurve=DataInterface() {propertyValues {Curve {keys [{arriveTangent=-53.9199, time=0, value=0}, {leaveTangent=0.2723, time=0.25, value=0.2}]}}})
    Index 0 Param 4=&FloatFromCurve(FloatCurve=DataInterface() {propertyValues {Curve {keys [<at 0 1>,<at 1 0>]}}})
    Index 1 Param 1=&FloatFromCurve(FloatCurve=DataInterface() {propertyValues {Curve {keys [<at 0 0>, {arriveTangent=4.2257, interpMode=RCIM_Cubic, leaveTangent=4.2257, tangentMode=RCTM_User, time=1, value=1}]}}})
    Index 1 Param 2=&FloatFromCurve(FloatCurve=DataInterface() {propertyValues {Curve {keys [<at 0 0>]}}})
    Index 1 Param 3=&FloatFromCurve(FloatCurve=DataInterface() {propertyValues {Curve {keys [{tangentMode=RCTM_User, time=0, value=0}, {tangentMode=RCTM_User, time=1, value=1}]}}})
    Index 1 Param 4=&FloatFromCurve(FloatCurve=DataInterface() {propertyValues {Curve {keys [{arriveTangent=-0.136, leaveTangent=-0.136, time=0, value=1}]}}}, Scale Curve=0.3)
    Index 2 Param 1=&FloatFromCurve(FloatCurve=DataInterface() {propertyValues {Curve {keys [
         {arriveTangent=2.646, interpMode=RCIM_Cubic, leaveTangent=2.646, tangentMode=RCTM_User, time=0, value=0}
         {leaveTangent=-0.0719, time=1, value=1}
        ]}}})
    Index 2 Param 2=&FloatFromCurve(
     FloatCurve=DataInterface() {propertyValues {Curve {keys [
         {arriveTangent=1.9056, time=0, value=0.2}
         {arriveTangent=0.0259, interpMode=RCIM_Cubic, leaveTangent=0.0259, tangentMode=RCTM_User, time=1, value=1.4}
        ]}}}
     Scale Curve=&RandomRangeFloat(Minimum=0.5, Maximum=1)
    )
    Index 2 Param 3=&FloatFromCurve(FloatCurve=DataInterface() {propertyValues {Curve {keys [
         {interpMode=RCIM_Cubic, leaveTangent=2.3064, tangentMode=RCTM_Break, time=0, value=1}
         {interpMode=RCIM_Cubic, leaveTangent=11.1233, tangentMode=RCTM_Break, time=1, value=2}
        ]}}})
    Index 2 Param 4=&FloatFromCurve(
     FloatCurve=DataInterface() {propertyValues {Curve {keys [
         {arriveTangent=-0.6867, interpMode=RCIM_Cubic, leaveTangent=-0.6867, tangentMode=RCTM_User, time=0, value=1}
         {interpMode=RCIM_Cubic, leaveTangent=-1.2956, tangentMode=RCTM_Break, time=1, value=0.75}
        ]}}}
     Scale Curve=1.6
    )
    Module.Param0WriteEnabled=true
    Module.Param0WriteEnabled1=true
    Module.Param0WriteEnabled2=true
    Module.Param1WriteEnabled=false
    Module.Param1WriteEnabled1=true
    Module.Param1WriteEnabled2=true
    Module.Param2WriteEnabled=true
    Module.Param2WriteEnabled1=true
    Module.Param2WriteEnabled2=true
    Module.Param3WriteEnabled=true
    Module.Param3WriteEnabled1=true
    Module.Param3WriteEnabled2=true
    Write Parameter Index 1=true
    Write Parameter Index 2=true
   )
   &Drag(Drag=3.2, Use Linear Drag=true, Use Rotational Drag=false)
   &CurlNoiseForce(Noise Strength=8000, Pan Noise=true, Pan Noise Field=<0 0 250>)
   &SolveForcesAndVelocity(Clamp Velocity=false, Limit Acceleration=false)
  ]
  rendererClasses [&NiagaraSpriteRendererProperties]
  renderers [
   {rendererClass=&NiagaraSpriteRendererProperties}
   {rendererClass=&NiagaraSpriteRendererProperties, rendererIndex=1}
  ]
 }
 FN_RadialMesh001 {
  emitterUpdateScript [
   &EmitterState()
   &SpawnRate(Use Spawn Probability=false) {enabled=false}
   &SpawnBurst_Instantaneous(Spawn Count=1, Use Spawn Probability=false)
  ]
  particleSpawnScript [
   &InitializeParticle(
    Lifetime=2
    Mesh Renderer Array Visibility Mode=::NewEnumerator1
    Mesh Scale Mode=::NewEnumerator3
    UsePositionOffset=false
    Write Material Random=true
   )
   /SetVariables() {isSetParameters=true}
  ]
  particleUpdateScript [
   &ParticleState()
   /SetVariables(Particles.SpriteSize=&Vector2DFromFloat(Value=&FloatFromCurve(
      FloatCurve=DataInterface() {propertyValues {Curve {keys [
          {interpMode=RCIM_Cubic, tangentMode=RCTM_User, time=0, value=-0.05}
          {arriveTangent=0.2231, interpMode=RCIM_Cubic, leaveTangent=0.2231, tangentMode=RCTM_User, time=0.1, value=1.125}
          {arriveTangent=0.8541, interpMode=RCIM_Cubic, leaveTangent=0.8541, tangentMode=RCTM_User, time=0.5, value=2}
          {arriveTangent=-12.9245, interpMode=RCIM_Cubic, tangentMode=RCTM_Break, time=0.5, value=0}
         ]}}}
      Scale Curve=500
     ))) {isSetParameters=true, module=SetVariables_#1}
   &FN_Radial_Material(
    Base Sprite Tile U=2
    Base Sprite Tile V=0.6
    Radial Mask Power=1
    Radial Mask Tile U=0.2
    Radial Mask Tile V=0.5
    Sprite Extrusion Speed:NiagaraFloat=&FloatFromCurve(
     FloatCurve=DataInterface() {propertyValues {Curve {keys [
         {arriveTangent=-0.2256, interpMode=RCIM_Cubic, time=0, value=1.75}
         {arriveTangent=8.4035, interpMode=RCIM_Cubic, leaveTangent=8.4035, time=0.302, value=3.125}
         {arriveTangent=16.5135, interpMode=RCIM_Cubic, leaveTangent=16.5135, time=0.4314, value=5.375}
         {interpMode=RCIM_Cubic, leaveTangent=-0.5666, time=0.5064, value=6.5}
        ]}}}
     Scale Curve=0.01
    )
    Sprite Spin Speed:NiagaraFloat=&FloatFromCurve(FloatCurve=DataInterface() {propertyValues {Curve {keys [<at 0 0.2>,<at 0.4 1.125>,<at 0.9528 2.875>]}}}, Scale Curve=-0.1)
   )
   &ScaleMeshSize(Scale Factor=&Multiply_VectorByFloat(
     Vector=<1 1 1>
     Float=&FloatFromCurve(
      FloatCurve=DataInterface() {propertyValues {Curve {keys [
          {interpMode=RCIM_Cubic, leaveTangentWeight=0.3333, tangentMode=RCTM_User, time=0, value=-0.05}
          {
           arriveTangent=5.2674
           arriveTangentWeight=0.0333
           interpMode=RCIM_Cubic
           leaveTangent=5.2674
           leaveTangentWeight=0.3
           tangentMode=RCTM_User
           time=0.1
           value=1
          }
          {
           arriveTangent=-2.068
           arriveTangentWeight=0.1
           interpMode=RCIM_Cubic
           leaveTangent=-2.068
           leaveTangentWeight=0.2
           tangentMode=RCTM_User
           time=0.4
           value=1.65
          }
          {
           arriveTangent=-12.9245
           arriveTangentWeight=0.0333
           interpMode=RCIM_Cubic
           leaveTangentWeight=0.1667
           tangentMode=RCTM_Break
           time=0.5
           value=0
          }
         ]}}}
      Scale Curve=10
     )
    ))
   &DynamicMaterialParameters(
    Index 3 Param 2=&FloatFromCurve(FloatCurve=DataInterface() {propertyValues {Curve {keys [
         {arriveTangent=0.2386, interpMode=RCIM_Cubic, leaveTangent=0.2386, tangentMode=RCTM_User, time=0.3649, value=-0.125}
         {arriveTangent=18.9086, interpMode=RCIM_Cubic, leaveTangent=18.9086, tangentMode=RCTM_User, time=0.5325, value=0.875}
        ]}}})
    Index 3 Param 3=0.6
    Module.Param0WriteEnabled3=false
    Module.Param1WriteEnabled3=true
    Module.Param2WriteEnabled3=true
    Module.Param3WriteEnabled3=false
    Write Parameter Index 0=false
    Write Parameter Index 3=true
   )
   /SetVariables(Particles.LightColor=&ScaleLinearColorByCurve(
     InitialValue=<rgba 0.2 3 5 5>
     ScaleCurve=DataInterface() {propertyValues {
       AlphaCurve {keys [<at 0 0.8>,<at 0.1 0.8>,<at 0.2 0.1>]}
       BlueCurve {keys [<at 0 1>,<at 0.1 0.8>,<at 0.2 0.1>]}
       GreenCurve {keys [<at 0 1>,<at 0.1 0.8>,<at 0.2 0.1>,<at 0.8 0.1>]}
       RedCurve {keys [<at 0 1>,<at 0.1 0.8>,<at 0.2 0.1>]}
      }}
     CurveIndex=$Particles.NormalizedAge
    )) {enabled=false, isSetParameters=true}
  ]
  rendererClasses [
   &NiagaraSpriteRendererProperties
   &NiagaraMeshRendererProperties
   &NiagaraLightRendererProperties
  ]
  renderers [
   {rendererClass=&NiagaraSpriteRendererProperties}
   {rendererClass=&NiagaraMeshRendererProperties, rendererIndex=1}
   {rendererClass=&NiagaraLightRendererProperties, rendererIndex=2}
  ]
 }
 FN_Starburst {
  emitterUpdateScript [
   &EmitterState()
   &SpawnRate(Use Spawn Probability=false) {enabled=false}
   &SpawnBurst_Instantaneous(Spawn Count=1, Spawn Time=0.1, Use Spawn Probability=false)
  ]
  particleSpawnScript [&InitializeParticle(Color Mode=::NewEnumerator0, Lifetime=2, UsePositionOffset=false, Write Material Random=true)]
  particleUpdateScript [
   &ParticleState()
   &Color(Color=&ColorFromCurve(ColorCurve=DataInterface() {propertyValues {
       AlphaCurve {keys [<at 0.3503 1 RCIM_Cubic>,<at 1 0 RCIM_Cubic>]}
       BlueCurve {keys [
         <at 0.0549 0.5573>
         {arriveTangent=0.0647, interpMode=RCIM_Cubic, leaveTangent=0.0647, time=0.6013, value=0.3646}
         <at 0.7712 0.6036 RCIM_Cubic>
        ]}
       GreenCurve {keys [<at 0.0549 0.1205>, {arriveTangent=1.2277, interpMode=RCIM_Cubic, leaveTangent=1.2277, time=0.6013, value=0.2291}, <at 0.7712 1 RCIM_Cubic>]}
       RedCurve {keys [
         <at 0.0549 0.0377>
         {arriveTangent=1.0388, interpMode=RCIM_Cubic, leaveTangent=1.0388, time=0.6013, value=0.1359}
         <at 0.7712 0.7819 RCIM_Cubic>
        ]}
      }})) {enabled=false}
   /SetVariables(Particles.SpriteSize=&Vector2DFromFloat(Value=&FloatFromCurve(
      FloatCurve=DataInterface() {propertyValues {Curve {keys [
          {interpMode=RCIM_Cubic, tangentMode=RCTM_User, time=0.2, value=0}
          {arriveTangent=12.4172, interpMode=RCIM_Cubic, leaveTangent=12.4172, tangentMode=RCTM_User, time=0.3, value=1.05}
          {arriveTangent=-2.068, interpMode=RCIM_Cubic, leaveTangent=-2.068, tangentMode=RCTM_User, time=0.4, value=1.975}
          {arriveTangent=-12.9245, interpMode=RCIM_Cubic, tangentMode=RCTM_Break, time=0.5058, value=0}
         ]}}}
      Scale Curve=1000
     ))) {isSetParameters=true, module=SetVariables_#1}
   &CameraOffset(Camera Offset Amount=-150) {enabled=false}
   &FN_Radial_Material(
    Radial Mask Power=1
    Radial Mask Tile U=0.2
    Radial Mask Tile V=0.5
    Sprite Extrusion Speed:NiagaraFloat=&FloatFromCurve(
     FloatCurve=DataInterface() {propertyValues {Curve {keys [
         {arriveTangent=-0.2256, interpMode=RCIM_Cubic, time=0, value=1.75}
         {arriveTangent=2.75, interpMode=RCIM_Cubic, leaveTangent=2.75, time=0.3, value=2.375}
         {arriveTangent=2.5, interpMode=RCIM_Cubic, leaveTangent=2.5, time=0.5, value=3.125}
         {interpMode=RCIM_Cubic, leaveTangent=-0.5666, time=0.7, value=3.375}
        ]}}}
     Scale Curve=-0.3
    )
    Sprite Spin Speed:NiagaraFloat=&FloatFromCurve(FloatCurve=DataInterface() {propertyValues {Curve {keys [<at 0 0.2>,<at 0.4 1.125>,<at 1 2.25>]}}}, Scale Curve=0.1)
   ) {enabled=false}
  ]
  rendererClasses [&NiagaraSpriteRendererProperties]
  renderers [
   {rendererClass=&NiagaraSpriteRendererProperties}
   {rendererClass=&NiagaraSpriteRendererProperties, rendererIndex=1}
   {rendererClass=&NiagaraSpriteRendererProperties, rendererIndex=2}
  ]
 }
 NE_FN_Sparks {
  simTarget=GPUComputeSim
  emitterUpdateScript [
   &EmitterState()
   &SpawnBurst_Instantaneous(Spawn Count=15, Spawn Time=0.3, Use Spawn Probability=false)
  ]
  particleSpawnScript [
   /SetVariables(Particles.SparkIntensity:NiagaraFloat=&FloatFromCurve(
     FloatCurve=DataInterface() {propertyValues {Curve {keys [
         <at 0 1 RCIM_Cubic>
         {arriveTangent=-2.5317, interpMode=RCIM_Cubic, leaveTangent=-2.5317, tangentMode=RCTM_User, time=0.6, value=0.5}
         <at 1 0 RCIM_Cubic>
        ]}}}
     CurveIndex=&RandomRangeFloat(Minimum=0, Maximum=1)
    )) {isSetParameters=true}
   &InitializeParticle(
    Color=&LerpLinearColors(StartColor=<rgba 2 4 5 1>, EndColor=<rgba 4 3.7291 0.3133 1>, LerpFactor=$Particles.SparkIntensity)
    Lifetime Max=1.5
    Lifetime Min=0.17
    Lifetime Mode=::NewEnumerator1
    Mass=&Lerp_Float(A=0.3, B=1, Alpha=$Particles.SparkIntensity)
    Mass Mode=::NewEnumerator1
    Sprite Size Mode=::NewEnumerator3
    Uniform Sprite Size=&Lerp_Float(A=25, B=25, Alpha=$Particles.SparkIntensity)
    UsePositionOffset=false
    Write Material Random=true
   )
   /SetVariables(Particles.SpriteSize=&Multiply_Vector2(
     A=$Particles.SpriteSize
     B=&Vector2DFromFloat(Value=&FloatFromCurve(FloatCurve=DataInterface() {propertyValues {Curve {keys [<at 1000 1 RCIM_Cubic>,<at 5000 2>]}}}, CurveIndex=$Engine.Owner.LODDistance))
    )) {isSetParameters=true}
   &SphereLocation(
    MaskSpawnGroup=false
    Override Seed=false
    Sphere Radius=&RandomRangeFloat(Minimum=10, Maximum=15)
    Surface Only=false
    UseNonUniformScale=false
    UseOffset=false
    UseSphereOrientation=false
   )
   &self:ScratchModule(CenterScale=600, velocity:Vector3f=<0 0 2500>) {module=ScratchModule}
   &AddVelocityInCone(
    Cone Angle=75
    Module.Use Velocity Falloff On Cone Axis=false
    Velocity Strength=&RandomRangeFloat(Minimum=500, Maximum=1500)
   )
  ]
  particleUpdateScript [
   &ParticleState()
   &GravityForce(Gravity=<0 0 -4000>)
   &Drag(Drag=3, Use Linear Drag=true, Use Rotational Drag=false)
   &VectorNoiseForce(
    Force Amount=&FloatFromCurve(
     FloatCurve=DataInterface() {propertyValues {Curve {keys [
         {arriveTangent=-9.4515, interpMode=RCIM_Cubic, leaveTangent=-4.1687, tangentMode=RCTM_User, time=0, value=1}
         {arriveTangent=-3.4372, interpMode=RCIM_Cubic, leaveTangent=-3.4372, tangentMode=RCTM_User, time=0.6, value=0}
        ]}}}
     Scale Curve=3000
    )
    Module.Use Random Noise Time Interval=false
   ) {enabled=false}
   &CurlNoiseForce(Noise Strength=8000, Pan Noise=false)
   &SolveForcesAndVelocity(Clamp Velocity=false, Limit Acceleration=false)
   &ScaleColor(
    Linear Color Curve=DataInterface() {propertyValues {
      AlphaCurve {keys [<at 0 1>,<at 1 1>]}
      BlueCurve {keys [
        <at 0 2 RCIM_Cubic>
        {arriveTangent=-1.452, interpMode=RCIM_Cubic, leaveTangent=-1.452, tangentMode=RCTM_User, time=0.1467, value=1}
        <at 1 0.25>
       ]}
      GreenCurve {keys [
        <at 0 2 RCIM_Cubic>
        {arriveTangent=-1.4408, interpMode=RCIM_Cubic, leaveTangent=-1.4408, tangentMode=RCTM_User, time=0.1467, value=1}
        <at 1 0.0972>
       ]}
      RedCurve {keys [
        <at 0 2 RCIM_Cubic>
        {arriveTangent=-0.8332, interpMode=RCIM_Cubic, leaveTangent=-0.8332, tangentMode=RCTM_User, time=0.1467, value=1}
        <at 1 0.0982>
       ]}
     }}
    Scale Mode=::NewEnumerator2
   )
   &ScaleSpriteSize(Scale Factor=&Vector2DFromFloat(Value=&FloatFromCurve(
      FloatCurve=DataInterface() {propertyValues {Curve {keys [<at 0.5 0.8 RCIM_Cubic>, {arriveTangent=0.2337, interpMode=RCIM_Cubic, leaveTangent=0.2337, tangentMode=RCTM_User, time=0.6, value=0.4}]}}}
      Scale Curve=0.4
     )))
   &ScaleSpriteSizeBySpeed(
    Max Scale Factor=<1.25 3>
    Sample Scale Factor By Curve=true
    Scale Factor Curve=DataInterface() {propertyValues {Curve {keys [
        {arriveTangent=0.2757, time=0, value=0}
        {arriveTangent=0.3844, interpMode=RCIM_Cubic, leaveTangent=0.3844, tangentMode=RCTM_User, time=0.3, value=0.15}
        {arriveTangent=2.2765, interpMode=RCIM_Cubic, leaveTangent=2.2765, tangentMode=RCTM_User, time=1, value=1}
       ]}}}
   )
  ]
  rendererClasses [&NiagaraSpriteRendererProperties]
  renderers [{rendererClass=&NiagaraSpriteRendererProperties}]
 }
 WispySprites {
  emitterUpdateScript [
   &EmitterState(Life Cycle Mode=::NewEnumerator1, Loop Behavior=::NewEnumerator1, Loop Duration=2, UseLoopDelay=false)
   &SpawnBurst_Instantaneous(Spawn Count=8, Spawn Time=0.1, Use Spawn Probability=false)
  ]
  particleSpawnScript [
   &InitializeParticle(
    Lifetime=&RandomRangeFloat(Minimum=0.3, Maximum=1)
    Sprite Size Mode=::NewEnumerator3
    Uniform Sprite Size=1200
    UsePositionOffset=false
    Write Material Random=true
   )
   &self:ScratchModule(
    CameraPosition:Vector3f=&ReturnCameraVector(Camera Property=::NewEnumerator0, Coordinate Space=::Simulation)
    CenterScale=700
    velocity:Vector3f=<0 0 300>
   ) {module=ScratchModule}
   &AddVelocityFromPoint(Module.Use Falloff Distance=false, Origin Offset Coordinate Space=::Simulation, Velocity Strength=300)
  ]
  particleUpdateScript [
   &ParticleState()
   &ScaleColor(
    Scale Alpha=&FloatFromCurve(FloatCurve=DataInterface() {propertyValues {Curve {keys [<at -0.0046 1.0113 RCIM_Cubic>,<at 1.0011 0.002 RCIM_Cubic>]}}})
    ScaleA=true
    ScaleRGB=false
   ) {enabled=false}
   &SolveForcesAndVelocity(Clamp Velocity=false, Limit Acceleration=false)
   &Color(Color=&ColorFromCurve(ColorCurve=DataInterface() {propertyValues {
       AlphaCurve {keys [<at 0.0136 0>,<at 0.0489 1 RCIM_Cubic>,<at 0.5621 1>,<at 1 0 RCIM_Cubic>]}
       BlueCurve {keys [<at 0.0706 0.401 RCIM_Cubic>, {interpMode=RCIM_Cubic, leaveTangent=0.3012, time=0.2362, value=5}, <at 0.4996 0.5>]}
       GreenCurve {keys [<at 0.0706 0.1867 RCIM_Cubic>, {interpMode=RCIM_Cubic, leaveTangent=2.4752, time=0.2362, value=2.4787}, <at 0.4996 1>]}
       RedCurve {keys [<at 0.0706 0 RCIM_Cubic>,<at 0.2362 0.3622 RCIM_Cubic>,<at 0.4996 0>]}
      }}))
   /SetVariables(Particles.SpriteSize=&Vector2DFromFloat(Value=&FloatFromCurve(
      FloatCurve=DataInterface() {propertyValues {Curve {keys [<at 0 1.7 RCIM_Cubic>, {arriveTangent=1.0647, interpMode=RCIM_Cubic, leaveTangent=1.0647, tangentMode=RCTM_User, time=2.5, value=2.95}]}}}
      Scale Curve=600
     ))) {isSetParameters=true}
   &SubUVAnimation() {enabled=false}
   &SpriteFacingAndAlignment(Sprite Alignment=$Particles.Velocity, WriteAlignment=true, WriteFacing=true)
  ]
  rendererClasses [&NiagaraSpriteRendererProperties]
  renderers [{rendererClass=&NiagaraSpriteRendererProperties}]
 }
}
summary {
 systemName=NS_Glider_Linguini
 emitters [
  {emitterName=FN_Starburst, rendererClasses [&NiagaraSpriteRendererProperties]}
  {
   emitterName=FN_RadialMesh001
   rendererClasses [
    &NiagaraSpriteRendererProperties
    &NiagaraMeshRendererProperties
    &NiagaraLightRendererProperties
   ]
  }
  {emitterName=WispySprites, rendererClasses [&NiagaraSpriteRendererProperties]}
  {emitterName=FN_Core_Smoke_Sprite, rendererClasses [&NiagaraSpriteRendererProperties]}
  {emitterName=NE_FN_Sparks, simTarget=GPUComputeSim, rendererClasses [&NiagaraSpriteRendererProperties]}
 ]
 userVariables []
}
