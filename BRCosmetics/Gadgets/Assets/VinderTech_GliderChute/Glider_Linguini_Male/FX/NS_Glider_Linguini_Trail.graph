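# Niagara trail effect for the Linguini male glider, with two anim-trail ribbons spanning bone or socket pairs.
# A color-graded ribbon and a fainter, camera-offset smoke ribbon share a lifetime fade multiplied by a curve of the system integrator value.
# Speedline sprites spawn at a ForwardDot-scaled rate and stretch with speed, while front-end switches adjust velocities and speedline behavior for menus.
# The contrail ribbon and light burst emitters are disabled.
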NiagaraSystem /BRCosmetics/Gadgets/Assets/VinderTech_GliderChute/Glider_Linguini_Male/FX/NS_Glider_Linguini_Trail
dependencies {
 usedDataInterfaces [
  &NiagaraDataInterfaceCurve
  &NiagaraDataInterfaceVectorCurve
  &NiagaraDataInterfaceColorCurve
  &NiagaraDataInterfaceVector2DCurve
  &NiagaraDataInterfaceMeshRendererInfo
 ]
 usedDynamicInputs []
 usedModules [&SystemState, &FN_Integrator]
}
emitters {
 AnimTrail {
  emitterUpdateScript [
   &EmitterState(Scalability Mode=::NewEnumerator1)
   /Game/Effects/Niagara/Modules/AnimTrail/FN_SpawnAnimTrail/FN_SpawnAnimTrail(Number of Anim Trails=2, SpawnRate=100)
  ]
  particleSpawnScript [
   /Game/Effects/Niagara/Modules/AnimTrail/FN_InitializeAnimTrail/FN_InitializeAnimTrailBoneIndecies(Second Bone or Socket Index=1)
   &FN_InitializeAnimTrailBoneIndecies(AnimTrail Selection=1, First Bone or Socket Index=2, Second Bone or Socket Index=3) {module=FN_InitializeAnimTrailBoneIndecies001}
   /Game/Effects/Niagara/Modules/AnimTrail/FN_InitializeAnimTrail/FN_InitializeAnimTrail(
    AnimTrailWidth=1
    Color=&MultiplyLinearColor(Linear Color=<rgba 1 1 1 1>, Scale Factor=<1 1 1 1>)
    Lifetime=0.175
    Write Material Random=false
   )
   &AddVelocity(Velocity=&Lerp_Vector(
     A=<0 0 0>
     B=<200 1500 0>
     Alpha=&MakeCustomFloatFromBool(Bool=$User.bisFrontEnd, True Float=1, False Float=0)
    ))
  ]
  particleUpdateScript [
   &InheritVelocity(Module.Limit Speed=true, Velocity Limit=20)
   &SolveForcesAndVelocity(Clamp Velocity=false, Limit Acceleration=false)
   /Game/Effects/Niagara/Modules/AnimTrail/FN_UpdateAnimTrail/FN_UpdateAnimTrail(New float=1)
   &ScaleColor(
    Scale Alpha=&Multiply_Float(
     A=&FloatFromCurve(FloatCurve=DataInterface() {propertyValues {Curve {keys [
          {arriveTangent=3.6646, interpMode=RCIM_Cubic, leaveTangent=3.6646, tangentMode=RCTM_User, time=-0.0017, value=0.5086}
          {arriveTangent=-0.0143, interpMode=RCIM_Cubic, leaveTangent=-0.0143, tangentMode=RCTM_User, time=0.1233, value=0.7411}
          {arriveTangent=-1.317, interpMode=RCIM_Cubic, leaveTangent=-1.317, tangentMode=RCTM_User, time=1, value=0}
         ]}}})
     B=&FloatFromCurve(
      FloatCurve=DataInterface() {propertyValues {Curve {keys [<at 0 0 RCIM_Cubic>, {arriveTangent=2.0248, interpMode=RCIM_Cubic, leaveTangent=2.0248, tangentMode=RCTM_User, time=1, value=1}]}}}
      CurveIndex=$System.FN_Integrator.Value
     )
    )
    Scale RGB=&MakeVectorFromLinearColorRGB(Color=&ColorFromCurve(ColorCurve=DataInterface() {propertyValues {
        AlphaCurve {keys [<at 0 1 RCIM_Cubic>]}
        BlueCurve {keys [<at 0.05 1>,<at 0.3 0.9276>,<at 0.5 1>,<at 1 1>]}
        GreenCurve {keys [<at 0.05 0.026>,<at 0.3 0.0918>,<at 0.5 0.2027>,<at 1 0.8323>]}
        RedCurve {keys [<at 0.05 0.789>,<at 0.3 0.1615>,<at 0.5 0.1302>,<at 1 0.1615>]}
       }}))
    ScaleA=true
    ScaleRGB=true
   )
  ]
  rendererClasses [&NiagaraRibbonRendererProperties]
  renderers [{rendererClass=&NiagaraRibbonRendererProperties}]
 }
 AnimTrail_Smoke {
  emitterUpdateScript [
   &EmitterState(Scalability Mode=::NewEnumerator1)
   &FN_SpawnAnimTrail(Number of Anim Trails=2, SpawnRate=100)
  ]
  particleSpawnScript [
   &FN_InitializeAnimTrailBoneIndecies(Second Bone or Socket Index=1)
   &FN_InitializeAnimTrailBoneIndecies(AnimTrail Selection=1, First Bone or Socket Index=2, Second Bone or Socket Index=3) {module=FN_InitializeAnimTrailBoneIndecies001}
   &FN_InitializeAnimTrail(
    AnimTrailWidth=1
    Color=&MultiplyLinearColor(Linear Color=<rgba 1 1 1 0.5>, Scale Factor=<1 1 1 1>)
    Lifetime=0.12
    Write Material Random=false
   )
   &CameraOffset(Camera Offset Amount=&RandomRangeFloat(Minimum=5, Maximum=15))
   &AddVelocity(Velocity=&Lerp_Vector(
     A=<0 0 0>
     B=<200 1500 0>
     Alpha=&MakeCustomFloatFromBool(Bool=$User.bisFrontEnd, True Float=1, False Float=0)
    ))
  ]
  particleUpdateScript [
   &InheritVelocity(Module.Limit Speed=true, Velocity Limit=20)
   &SolveForcesAndVelocity(Clamp Velocity=false, Limit Acceleration=false)
   &FN_UpdateAnimTrail(New float=1)
   &ScaleColor(
    Scale Alpha=&Multiply_Float(
     A=&FloatFromCurve(FloatCurve=DataInterface() {propertyValues {Curve {keys [
          {arriveTangent=3.6646, interpMode=RCIM_Cubic, leaveTangent=3.6646, tangentMode=RCTM_User, time=-0.0017, value=0.5086}
          {arriveTangent=-0.0143, interpMode=RCIM_Cubic, leaveTangent=-0.0143, tangentMode=RCTM_User, time=0.1233, value=0.7411}
          {arriveTangent=-1.317, interpMode=RCIM_Cubic, leaveTangent=-1.317, tangentMode=RCTM_User, time=1, value=0}
         ]}}})
     B=&FloatFromCurve(
      FloatCurve=DataInterface() {propertyValues {Curve {keys [<at 0 0 RCIM_Cubic>, {arriveTangent=2.0248, interpMode=RCIM_Cubic, leaveTangent=2.0248, tangentMode=RCTM_User, time=1, value=1}]}}}
      CurveIndex=$System.FN_Integrator.Value
     )
    )
    Scale RGB=&MakeVectorFromLinearColorRGB(Color=<rgba 1.4879 10 0 1>)
    ScaleA=true
    ScaleRGB=false
   )
  ]
  rendererClasses [&NiagaraRibbonRendererProperties]
  renderers [{rendererClass=&NiagaraRibbonRendererProperties}]
 }
 FN_Contrails_RibbonWde_pink {
  bEnabled=false
  emitterUpdateScript [
   &EmitterScalability(
    Module.bApplySpawnCountScale=false
    Module.bApplySpawnCountScaleByDistanceFraction=false
    Module.bUseMaxDistance=false
    Module.bUseMinDistance=false
   )
   &EmitterLifeCycle(CompleteOnInactive=true)
   &EmitterState()
   /Game/Effects/Niagara/Modules/Spawn/FN_SpawnContrails(All limbs=false, Global Spawnrate=30)
   &FN_SpawnRibbon_v2(Number of Ribbons=2, Spawn Mode=::NewEnumerator0, SpawnRate=20)
  ]
  particleSpawnScript [
   &FN_InitializeRibbon_v3(Enable Skeletal Position=true, Lifetime=1, Ribbon Width=16)
   &AddVelocity(Velocity=<0 0 410>) {enabled=false}
   &AddVelocity(Velocity=&FN_FrontEndVector3Modifier(Gameplay=<0 0 410>, Front End=<0 0 350>)) {module=AddVelocity001}
  ]
  particleUpdateScript [
   &AccelerationForce(
    Acceleration=&VectorFromCurve(VectorCurve=DataInterface() {propertyValues {
       XCurve {keys [<at 0.0017 132.1921>,<at 0.9905 -65.3336>]}
       YCurve {keys [<at -0.0011 1.0067>,<at 0.9988 0>]}
       ZCurve {keys [<at -0.0011 -300>,<at 0.9957 -200.05>]}
      }})
    Coordinate Space=::Local
   ) {enabled=false}
   &Drag(Drag=0.3, Use Linear Drag=true, Use Rotational Drag=false)
   &InheritVelocity(
    Module.Limit Speed=true
    Velocity Limit=10000
    Velocity Scale=<-0.03 -0.03 -0.01>
    Velocity Source=&TransformVector(Vector=$Engine.Owner.Velocity, Source Space=::World, Destination Space=::Local)
   ) {enabled=false}
   /Game/Effects/Niagara/Modules/Skydiving/FN_DivingAttributes/FN_DivingAttributes()
   &CurlNoiseForce(Noise Frequency=5, Noise Strength=111, Pan Noise=false) {enabled=false}
   &SolveForcesAndVelocity(Clamp Velocity=false, Limit Acceleration=false)
   &FN_UpdateRibbon(
    Enable Skeletal Position=true
    Initial Color:LinearColor=$Particles.Initial.Color
    Initial RibbonWidth:NiagaraFloat=$Particles.Initial.RibbonWidth
   )
   &RotateAroundPoint(
    Derive the Rotation Axis=false
    Rotation Axis:Vector3f=$Particles.Velocity
    Rotation Center:NiagaraPosition=$Particles.Position
    Rotation Phase=2
    Rotation Rate=2
    X Axis:Vector3f=<0 1 0>
   )
   &Material(Base Texture Time:NiagaraFloat=&Multiply_Float(A=$Emitter.Age, B=0)) {enabled=false}
   &DynamicMaterialParameters(Module.Param0WriteEnabled=true, Module.Param1WriteEnabled=true, Module.Param2WriteEnabled=true, Module.Param3WriteEnabled=true) {enabled=false}
   &Color(
    Color=&ColorFromCurve(ColorCurve=DataInterface() {propertyValues {
       AlphaCurve {keys [<at 0.3766 1 RCIM_Cubic>,<at 0.9857 -0.003 RCIM_Cubic>]}
       BlueCurve {keys [<at -0.0023 0 RCIM_Cubic>,<at 0.3035 0>]}
       GreenCurve {keys [<at -0.0023 1 RCIM_Cubic>,<at 0.3035 0.1855>]}
       RedCurve {keys [<at -0.0023 0.1629 RCIM_Cubic>,<at 0.3035 0.7656>]}
      }})
    Scale Alpha=&FloatFromCurve(
     FloatCurve=DataInterface() {propertyValues {Curve {keys [{arriveTangent=0.27, interpMode=RCIM_Cubic, leaveTangent=0.27, tangentMode=RCTM_User, time=-0.0046, value=1.0113}, <at 0.6 0.002>]}}}
     Scale Curve=0.8
    )
   ) {enabled=false}
   &DynamicMaterialParameters(
    Index 0 Param 1=1
    Index 0 Param 2=1
    Index 0 Param 3=0.1
    Index 0 Param 4=0.2
    Index 1 Param 1=&Add_Float(
     A=&Multiply_Float(A=1, B=&MakeFloatFromInt(Int=HlslExpression() {hlslExpression=Particles.RibbonID}))
     B=&Multiply_Float(A=-1, B=$Engine.Time)
    )
    Index 1 Param 3=1
    Index 1 Param 4=1
    Index 2 Param 1=0.5
    Index 2 Param 2=0.5
    Index 2 Param 3=0.5
    Index 3 Param 1=1
    Index 3 Param 2=1
    Index 3 Param 3=1
    Index 3 Param 4=1
    Module.Param0WriteEnabled=true
    Module.Param0WriteEnabled1=true
    Module.Param0WriteEnabled2=true
    Module.Param0WriteEnabled3=true
    Module.Param1WriteEnabled=true
    Module.Param1WriteEnabled1=true
    Module.Param1WriteEnabled2=true
    Module.Param1WriteEnabled3=true
    Module.Param2WriteEnabled=true
    Module.Param2WriteEnabled1=true
    Module.Param2WriteEnabled2=true
    Module.Param2WriteEnabled3=true
    Module.Param3WriteEnabled=true
    Module.Param3WriteEnabled1=true
    Module.Param3WriteEnabled2=true
    Module.Param3WriteEnabled3=true
    Write Parameter Index 1=true
    Write Parameter Index 2=true
    Write Parameter Index 3=true
   ) {enabled=false, module=DynamicMaterialParameters001}
   &FN_RibbonMaterial(
    Color Pan U Speed=2
    Color Scale U=2.5
    Erosion Pan Speed U=10
    Erosion Power=0.3
    Opacity Power=1.5
    Scale Alpha:NiagaraFloat=&Multiply_Float(A=2, B=&OneMinusFloat(Float=&Power(a=$Particles.NormalizedAge, b=0.7)))
   )
  ]
  rendererClasses [&NiagaraRibbonRendererProperties]
  renderers [{rendererClass=&NiagaraRibbonRendererProperties}]
 }
 SimpleSpriteBurst {
  bEnabled=false
  emitterUpdateScript [
   &EmitterState(Life Cycle Mode=::NewEnumerator1, Loop Behavior=::NewEnumerator1, Loop Duration=2, UseLoopDelay=false)
   &SpawnBurst_Instantaneous(Use Spawn Probability=false, UseLoopCountLimit=false)
  ]
  particleSpawnScript [
   &InitializeParticle(
    Lifetime=2
    Mesh Renderer Array Visibility Mode=::NewEnumerator1
    Mesh Scale Mode=::NewEnumerator3
    UsePositionOffset=false
    Write Material Random=true
   )
   /SetVariables(Particles.Position=&Add_Vector(A=Linked() {linkedVariable {name=Particles.Position, type=NiagaraPosition}}, B=<0 5 5>)) {isSetParameters=true}
  ]
  particleUpdateScript [
   &ParticleState()
   /SetVariables(Particles.LightColor=&ScaleLinearColorByCurve(
     InitialValue=<rgba 0.1 1.5 2.5 1.5>
     ScaleCurve=DataInterface() {propertyValues {
       AlphaCurve {keys [<at 0 0.8>,<at 0.5521 0.8>,<at 0.856 0.1>]}
       BlueCurve {keys [<at 0 1>,<at 0.5521 0.8>,<at 0.8494 0.1>]}
       GreenCurve {keys [<at 0 1>,<at 0.5521 0.8>,<at 0.8 0.1>,<at 0.8494 0.1>]}
       RedCurve {keys [<at 0 1>,<at 0.5521 0.8>,<at 0.8494 0.1>]}
      }}
     CurveIndex=$Particles.NormalizedAge
    )) {isSetParameters=true}
  ]
  rendererClasses [&NiagaraLightRendererProperties]
  renderers [{rendererClass=&NiagaraLightRendererProperties, rendererIndex=0}]
 }
 Speedlines {
  emitterUpdateScript [
   &EmitterState(Enable Distance Culling=true, MaxDistance=2048, Module.bUseMaxDistance=true, Module.bUseMinDistance=false, Scalability Mode=::NewEnumerator1)
   &SpawnRate(
    SpawnRate=&Multiply_Float(
     A=&FN_FrontEndFloatModifier(Gameplay=$User.ForwardDot, Front End=1)
     B=&UniformRangedFloat(Minimum=75, Maximum=150)
    )
    Use Spawn Probability=false
   )
  ]
  particleSpawnScript [
   /Niagara/Modules/Spawn/Initialization/InitializeParticle(
    Color=&MakeLinearColorFromVectorAndFloat(
     "Vector (RGB)"=<1 1 1>
     "Float (Alpha)"=&Multiply_Float(
      A=&Multiply_Float(
       A=&UniformRangedFloat(Minimum=0.015, Maximum=0.15)
       B=&FN_FrontEndFloatModifier(Gameplay=1, Front End=0.5)
      )
      B=&Add_Float(
       A=1
       B=&Multiply_Float(A=10, B=&Subtract_Float(A=1, B=$Engine.Emitter.SpawnCountScale))
      )
     )
    )
    Lifetime=&FN_FrontEndFloatModifier(
     Gameplay=&UniformRangedFloat(Minimum=0.2, Maximum=0.35)
     Front End=&UniformRangedFloat(Minimum=0.1, Maximum=0.2)
    )
    Module.Write Color=true
    Module.Write Lifetime=true
    Module.Write Mass=true
    Module.Write Material Random=true
    Module.Write Position=false
    Module.Write Scale=false
    Module.Write SpriteRotation=false
    Module.Write SpriteSize=true
    Position=&SimulationPosition()
    Sprite Size=&UniformRangedVector2D(Minimum=<3 4>, Maximum=<5 6>)
   )
   &CylinderLocation(
    Cylinder Radius=&UniformRangedFloat(Minimum=100, Maximum=150)
    MaskSpawnGroup=false
    Module.Enable Lathe Profile=false
    Module.Override Randomness=false
    Module.Surface Only=true
    Orientation Axis:ENiagaraOrientationAxis=::XAxis
    Use Endcaps In Surface Only Mode=false
   )
   &AddVelocity(Velocity=&FN_FrontEndVector3Modifier(Gameplay=<0 1000 0>, Front End=<0 3000 0>))
   &AddVelocityFromPoint(
    Module.Use Falloff Distance=false
    Velocity Strength=&Multiply_Float(
     A=&FN_FrontEndFloatModifier(Gameplay=1, Front End=2)
     B=&UniformRangedFloat(Minimum=500, Maximum=700)
    )
   )
  ]
  particleUpdateScript [
   &ParticleState()
   &SolveForcesAndVelocity(Clamp Velocity=false, Limit Acceleration=false)
   &ScaleSpriteSize(Scale Factor=&Vector2DFromCurve(
     Vector2Curve=DataInterface() {propertyValues {XCurve {keys [<at 0 0.5 RCIM_Cubic>,<at 0.5 1 RCIM_Cubic>]}, YCurve {keys [<at 0 0 RCIM_Cubic>,<at 0.5 1 RCIM_Cubic>]}}}
     ScaleCurve=&Vector2DFromFloat(Value=1)
    ))
   &ScaleSpriteSizeBySpeed(Max Scale Factor=<1 40>, Min Scale Factor=<0.01 0.05>, Velocity Threshold=1)
   &ScaleColor(
    Scale Alpha=&FloatFromCurve(
     FloatCurve=DataInterface() {propertyValues {Curve {keys [<at 0 0 RCIM_Cubic>, <at 0.2 1>, {arriveTangent=-2.221, interpMode=RCIM_Cubic, leaveTangent=-2.221, tangentMode=RCTM_User, time=1, value=0}]}}}
     Scale Curve=&FN_FrontEndFloatModifier(Gameplay=0.4, Front End=0)
    )
    ScaleA=true
    ScaleRGB=true
   )
  ]
  rendererClasses [&NiagaraSpriteRendererProperties]
  renderers [{rendererClass=&NiagaraSpriteRendererProperties}]
 }
}
summary {
 systemName=NS_Glider_Linguini_Trail
 emitters [
  {bEnabled=false, emitterName=FN_Contrails_RibbonWde_pink, rendererClasses [&NiagaraRibbonRendererProperties]}
  {emitterName=AnimTrail, rendererClasses [&NiagaraRibbonRendererProperties]}
  {emitterName=AnimTrail_Smoke, rendererClasses [&NiagaraRibbonRendererProperties]}
  {emitterName=Speedlines, rendererClasses [&NiagaraSpriteRendererProperties]}
  {bEnabled=false, emitterName=SimpleSpriteBurst, rendererClasses [&NiagaraLightRendererProperties]}
 ]
 userVariables [
  User.bIsFrontEnd {type=NiagaraBool}
  User.bIsFrontEndPreview {type=NiagaraBool}
  User.ForwardDot {type=NiagaraFloat, defaultValue=0, description=""}
  User.isFullyDeployed {type=NiagaraBool}
  User.RibbonIDOffset {type=NiagaraInt32}
  User.RightDot {type=NiagaraFloat, defaultValue=0, description=""}
  User.SkeletalMesh {
   type=NiagaraDataInterfaceSkeletalMesh
   defaultValue {
    struct=NiagaraExt_VariableValue_DataInterface
    value {dataInterface=&self:NiagaraDataInterfaceSkeletalMesh_0}
   }
  }
 ]
}
